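# Physics asset for secondary motion on the torso of the male medium TheBomb character body, covering the right collar, zipper and left back strap.
# Each piece constrains to a tiny capsule MSTR anchor body, with the strap forming a two-box chain.
# Angular limits shape the motion: the collar swings on one axis only, the strap bends mostly on one axis, and the zipper swings more freely with little twist.
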PhysicsAsset /BRCosmetics/Characters/Player/Male/Medium/Bodies/M_MED_TheBomb/Meshes/M_MED_TheBomb_Dyn_Torso
bodies {
 dyn_collar_r [{shapeType=Capsule, length=3, radius=2, center=<1.7703 0 0>, rotation=<rot 90 0 0>}]
 dyn_collar_r_MSTR [{shapeType=Capsule, length=0.505, radius=0.505}]
 dyn_strap_back_l_01 [{shapeType=Box, extentX=6, extentY=4.5, extentZ=1, center=<1.3326 -0.0001 0>}]
 dyn_strap_back_l_02 [{shapeType=Box, extentX=6, extentY=4.5, extentZ=1, center=<1.9059 -0.0001 0>}]
 dyn_strap_back_l_03_MSTR [{shapeType=Capsule, length=0.505, radius=0.505}]
 dyn_zipper [{shapeType=Capsule, length=3, radius=2, center=<1.9739 0 0>, rotation=<rot 90 0 0>}]
 dyn_zipper_MSTR [{shapeType=Capsule, length=0.505, radius=0.505}]
}
constraints [
 {bone1Name=dyn_strap_back_l_02, bone2Name=dyn_strap_back_l_01, swing1LimitDegrees=5, swing2LimitDegrees=45, twistLimitDegrees=1}
 {bone1Name=dyn_strap_back_l_01, bone2Name=dyn_strap_back_l_03_MSTR, swing1LimitDegrees=5, swing2LimitDegrees=45, twistLimitDegrees=1}
 {bone1Name=dyn_zipper, bone2Name=dyn_zipper_MSTR, swing1LimitDegrees=45, swing2LimitDegrees=45, twistLimitDegrees=2}
 {
  bone1Name=dyn_collar_r
  bone2Name=dyn_collar_r_MSTR
  swing1LimitDegrees=2
  swing1Motion=Locked
  swing2LimitDegrees=45
  twistLimitDegrees=1
  twistMotion=Locked
 }
]
