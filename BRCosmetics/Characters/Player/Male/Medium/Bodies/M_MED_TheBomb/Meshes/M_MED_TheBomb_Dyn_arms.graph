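# Physics asset for the arm dynamics of the male medium TheBomb cosmetic body, covering wrist chains, sleeves and a left forearm strap.
# Wrist chains (three links left, four right) run from master bones that rotate freely on the lower-arm twist bones.
# Three sleeve bodies per side hang from upper-arm twist bones with limited swing.
# The two-link strap has tight limits; its master has no parent constraint, and the hand capsules are unconstrained.
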PhysicsAsset /BRCosmetics/Characters/Player/Male/Medium/Bodies/M_MED_TheBomb/Meshes/M_MED_TheBomb_Dyn_arms
bodies {
 dyn_chain_wrist_l_01 [{shapeType=Capsule, length=3, radius=2, center=<-2.7644 0 0.0001>, rotation=<rot 90 0 0>}]
 dyn_chain_wrist_l_02 [{shapeType=Capsule, length=3, radius=2, center=<-2.7644 0 0.0001>, rotation=<rot 90 0 0>}]
 dyn_chain_wrist_l_03 [{shapeType=Capsule, length=3, radius=2, center=<-2.7644 0 0.0001>, rotation=<rot 90 0 0>}]
 dyn_chain_wrist_l_MSTR [{shapeType=Capsule, length=0.505, radius=0.505}]
 dyn_chain_wrist_r_01 [{shapeType=Capsule, length=3, radius=2, center=<2.7644 0 0.0001>, rotation=<rot 90 0 0>}]
 dyn_chain_wrist_r_02 [{shapeType=Capsule, length=3, radius=2, center=<2.7644 0 0.0001>, rotation=<rot 90 0 0>}]
 dyn_chain_wrist_r_03 [{shapeType=Capsule, length=3, radius=2, center=<2.7644 0 0.0001>, rotation=<rot 90 0 0>}]
 dyn_chain_wrist_r_04 [{shapeType=Capsule, length=3, radius=2, center=<2.7644 0 0.0001>, rotation=<rot 90 0 0>}]
 dyn_chain_wrist_r_MSTR [{shapeType=Capsule, length=0.505, radius=0.505}]
 dyn_sleeve_bk_l [{shapeType=Capsule, length=3, radius=2, center=<0.4 0 0>, rotation=<rot 90 0 0>}]
 dyn_sleeve_bk_r [{shapeType=Capsule, length=3, radius=2, center=<-0.4 0 0>, rotation=<rot 90 0 0>}]
 dyn_sleeve_in_l [{shapeType=Capsule, length=3, radius=2, center=<0.4 0 0>, rotation=<rot 90 0 0>}]
 dyn_sleeve_in_r [{shapeType=Capsule, length=3, radius=2, center=<-0.4 0 0>, rotation=<rot 90 0 0>}]
 dyn_sleeve_out_l [{shapeType=Capsule, length=3, radius=2, center=<0.4 0 0>, rotation=<rot 90 0 0>}]
 dyn_sleeve_out_r [{shapeType=Capsule, length=3, radius=2, center=<-0.4 0 0>, rotation=<rot 90 0 0>}]
 dyn_strap_forearm_l_01 [{shapeType=Box, extentX=4.5, extentY=3.5, extentZ=1, center=<0.3428 -0.0003 0>}]
 dyn_strap_forearm_l_02 [{shapeType=Box, extentX=2.5, extentY=3.5, extentZ=1, center=<0.2396 -0.0001 0.0001>}]
 dyn_strap_forearm_l_MSTR [{shapeType=Capsule, length=0.505, radius=0.505}]
 hand_l [{shapeType=Capsule, length=5.1515, radius=4.4915, center=<4.8364 -1.3321 0.4707>, rotation=<rot -4.3548 -89.9995 82.1834>}]
 hand_r [{shapeType=Capsule, length=5.1515, radius=4.4915, center=<-4.8364 1.3321 -0.4707>, rotation=<rot -4.3548 -89.9995 82.1834>}]
 lowerarm_twist_01_l [{shapeType=Capsule, length=5.25, radius=4.25, center=<-2.7091 0 0>, rotation=<rot 90 0 0>}]
 lowerarm_twist_01_r [{shapeType=Capsule, length=5.25, radius=4.25, center=<2.7091 0 0>, rotation=<rot 90 0 0>}]
 upperarm_twist_02_l [{shapeType=Capsule, length=3, radius=2, rotation=<rot 90 0 0>}]
 upperarm_twist_02_r [{shapeType=Capsule, length=3, radius=2, rotation=<rot 90 0 0>}]
}
constraints [
 {bone1Name=dyn_chain_wrist_l_02, bone2Name=dyn_chain_wrist_l_01, swing1LimitDegrees=60, swing2LimitDegrees=60, twistLimitDegrees=22}
 {bone1Name=dyn_chain_wrist_l_03, bone2Name=dyn_chain_wrist_l_02, swing1LimitDegrees=60, swing2LimitDegrees=60, twistLimitDegrees=22}
 {bone1Name=dyn_chain_wrist_l_01, bone2Name=dyn_chain_wrist_l_MSTR, swing1LimitDegrees=120, swing2LimitDegrees=120, twistLimitDegrees=22}
 {bone1Name=dyn_strap_forearm_l_02, bone2Name=dyn_strap_forearm_l_01, swing1LimitDegrees=5, swing2LimitDegrees=45, twistLimitDegrees=1}
 {bone1Name=dyn_strap_forearm_l_01, bone2Name=dyn_strap_forearm_l_MSTR, swing1LimitDegrees=5, swing2LimitDegrees=45, twistLimitDegrees=1}
 {bone1Name=dyn_sleeve_out_l, bone2Name=upperarm_twist_02_l, swing1LimitDegrees=20, swing2LimitDegrees=45, twistLimitDegrees=2}
 {bone1Name=dyn_sleeve_bk_l, bone2Name=upperarm_twist_02_l, swing1LimitDegrees=20, swing2LimitDegrees=45, twistLimitDegrees=2}
 {bone1Name=dyn_sleeve_in_l, bone2Name=upperarm_twist_02_l, swing1LimitDegrees=20, swing2LimitDegrees=45, twistLimitDegrees=2}
 {bone1Name=dyn_chain_wrist_r_04, bone2Name=dyn_chain_wrist_r_03, swing1LimitDegrees=60, swing2LimitDegrees=60, twistLimitDegrees=22}
 {bone1Name=dyn_chain_wrist_r_03, bone2Name=dyn_chain_wrist_r_02, swing1LimitDegrees=60, swing2LimitDegrees=60, twistLimitDegrees=22}
 {bone1Name=dyn_chain_wrist_r_02, bone2Name=dyn_chain_wrist_r_01, swing1LimitDegrees=60, swing2LimitDegrees=60, twistLimitDegrees=22}
 {bone1Name=dyn_chain_wrist_r_01, bone2Name=dyn_chain_wrist_r_MSTR, swing1LimitDegrees=120, swing2LimitDegrees=120, twistLimitDegrees=22}
 {bone1Name=dyn_sleeve_in_r, bone2Name=upperarm_twist_02_r, swing1LimitDegrees=20, swing2LimitDegrees=45, twistLimitDegrees=2}
 {bone1Name=dyn_sleeve_out_r, bone2Name=upperarm_twist_02_r, swing1LimitDegrees=20, swing2LimitDegrees=45, twistLimitDegrees=2}
 {bone1Name=dyn_sleeve_bk_r, bone2Name=upperarm_twist_02_r, swing1LimitDegrees=20, swing2LimitDegrees=45, twistLimitDegrees=2}
 {
  bone1Name=dyn_chain_wrist_l_MSTR
  bone2Name=lowerarm_twist_01_l
  swing1LimitDegrees=45
  swing1Motion=Free
  swing2LimitDegrees=45
  swing2Motion=Free
  twistLimitDegrees=45
  twistMotion=Free
 }
 {
  bone1Name=dyn_chain_wrist_r_MSTR
  bone2Name=lowerarm_twist_01_r
  swing1LimitDegrees=45
  swing1Motion=Free
  swing2LimitDegrees=45
  swing2Motion=Free
  twistLimitDegrees=45
  twistMotion=Free
 }
]
