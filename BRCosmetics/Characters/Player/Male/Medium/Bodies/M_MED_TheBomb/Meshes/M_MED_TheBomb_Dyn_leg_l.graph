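# Physics asset for the left-leg dynamic pant sleeve bones on the TheBomb medium male player cosmetic body.
# Four small capsule bodies on the back, front, inner and outer sleeve bones are each constrained to a capsule on thigh_l.
# All four joints share the same angular limits, allowing moderate swing and almost no twist.
# The sleeve bones carry _r suffixes despite attaching to the left thigh.
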PhysicsAsset /BRCosmetics/Characters/Player/Male/Medium/Bodies/M_MED_TheBomb/Meshes/M_MED_TheBomb_Dyn_leg_l
bodies {
 dyn_pant_sleeve_bk_r [{shapeType=Capsule, length=3, radius=2, center=<0.4 0 0>, rotation=<rot 90 0 0>}]
 dyn_pant_sleeve_fr_r [{shapeType=Capsule, length=3, radius=2, center=<0.4 0 0>, rotation=<rot 90 0 0>}]
 dyn_pant_sleeve_in_r [{shapeType=Capsule, length=3, radius=2, center=<0.4 0 0>, rotation=<rot 90 0 0>}]
 dyn_pant_sleeve_out_r [{shapeType=Capsule, length=3, radius=2, center=<0.4 0 0>, rotation=<rot 90 0 0>}]
 thigh_l [{shapeType=Capsule, length=3, radius=2, rotation=<rot 90 0 0>}]
}
constraints [
 {bone1Name=dyn_pant_sleeve_bk_r, bone2Name=thigh_l, swing1LimitDegrees=20, swing2LimitDegrees=45, twistLimitDegrees=2}
 {bone1Name=dyn_pant_sleeve_in_r, bone2Name=thigh_l, swing1LimitDegrees=20, swing2LimitDegrees=45, twistLimitDegrees=2}
 {bone1Name=dyn_pant_sleeve_out_r, bone2Name=thigh_l, swing1LimitDegrees=20, swing2LimitDegrees=45, twistLimitDegrees=2}
 {bone1Name=dyn_pant_sleeve_fr_r, bone2Name=thigh_l, swing1LimitDegrees=20, swing2LimitDegrees=45, twistLimitDegrees=2}
]
