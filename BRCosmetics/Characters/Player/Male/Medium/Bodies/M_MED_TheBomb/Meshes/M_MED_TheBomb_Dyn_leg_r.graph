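# Physics asset for the right leg of the M_MED_TheBomb player body cosmetic, covering its dangling ankle chain and inner and outer shoelace bones.
# Small capsules form a three-link ankle chain, a two-link inner lace and a single outer lace, each constrained link-to-link from its own MSTR root body with identical swing and twist limits.
# Two larger capsules on foot_r give the foot a collision shape, with no constraint tying it to the chains.
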PhysicsAsset /BRCosmetics/Characters/Player/Male/Medium/Bodies/M_MED_TheBomb/Meshes/M_MED_TheBomb_Dyn_leg_r
bodies {
 dyn_chain_ankle_r_01 [{shapeType=Capsule, length=3, radius=2, center=<2.7644 0 0.0001>, rotation=<rot 90 0 0>}]
 dyn_chain_ankle_r_02 [{shapeType=Capsule, length=3, radius=2, center=<2.7644 0 0.0001>, rotation=<rot 90 0 0>}]
 dyn_chain_ankle_r_03 [{shapeType=Capsule, length=3, radius=2, center=<2.7644 0 0.0001>, rotation=<rot 90 0 0>}]
 dyn_chain_ankle_r_MSTR [{shapeType=Capsule, length=0.505, radius=0.505}]
 dyn_shoelace_in_r_01 [{shapeType=Capsule, length=4.25, radius=0.75, center=<2.7644 0 0.0001>, rotation=<rot 90 0 0>}]
 dyn_shoelace_in_r_02 [{shapeType=Capsule, length=4.25, radius=0.75, center=<2.7644 0 0.0001>, rotation=<rot 90 0 0>}]
 dyn_shoelace_in_r_MSTR [{shapeType=Capsule, length=0.505, radius=0.505}]
 dyn_shoelace_out_r [{shapeType=Capsule, length=2.75, radius=0.75, center=<1.7558 0 0.0001>, rotation=<rot 90 0 0>}]
 dyn_shoelace_out_r_MSTR [{shapeType=Capsule, length=0.505, radius=0.505}]
 foot_r [
  {shapeType=Capsule, length=7.6127, radius=6.3065, center=<0 0.6517 0.5209>, rotation=<rot 90 0 0>}
  {shapeType=Capsule, length=8, radius=5.5, center=<2.3162 -5.0419 0.2611>, rotation=<rot 0 35 89.9999>}
 ]
}
constraints [
 {bone1Name=dyn_chain_ankle_r_01, bone2Name=dyn_chain_ankle_r_MSTR, swing1LimitDegrees=60, swing2LimitDegrees=60, twistLimitDegrees=22}
 {bone1Name=dyn_chain_ankle_r_02, bone2Name=dyn_chain_ankle_r_01, swing1LimitDegrees=60, swing2LimitDegrees=60, twistLimitDegrees=22}
 {bone1Name=dyn_chain_ankle_r_03, bone2Name=dyn_chain_ankle_r_02, swing1LimitDegrees=60, swing2LimitDegrees=60, twistLimitDegrees=22}
 {bone1Name=dyn_shoelace_in_r_01, bone2Name=dyn_shoelace_in_r_MSTR, swing1LimitDegrees=60, swing2LimitDegrees=60, twistLimitDegrees=22}
 {bone1Name=dyn_shoelace_in_r_02, bone2Name=dyn_shoelace_in_r_01, swing1LimitDegrees=60, swing2LimitDegrees=60, twistLimitDegrees=22}
 {bone1Name=dyn_shoelace_out_r, bone2Name=dyn_shoelace_out_r_MSTR, swing1LimitDegrees=60, swing2LimitDegrees=60, twistLimitDegrees=22}
]
